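# Sound cue for the kunai throwable's ambient loop, in the explosives sound class.
# A root mixer layers two looping ambient waves with a burning branch at differing input volumes.
# That branch waits a fixed one second, then plays a burning-start one-shot and a burning loop together.
# The cue runs at reduced volume with inline attenuation and concurrency overrides that set falloff and cap simultaneous instances.
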SoundCue /HeadbandGameplay/Sounds/Weapons/Throwable/Kunai/Kunai_Ambient_Loop_Cue
root=Mixer_0
nodes [
 {id=Delay_0, children [Mixer_1], properties {delayMax=1, delayMin=1}}
 {id=Mixer_0, children [WavePlayer_0,WavePlayer_3,Delay_0], properties {inputVolume [0.45,1,1.4]}}
 {id=Mixer_1, children [WavePlayer_1,WavePlayer_2], properties {inputVolume [1,1]}}
 {id=WavePlayer_0, properties {bLooping=true, soundWaveAssetPtr=&/Kunai_Ambient_Loop_04}}
 {id=WavePlayer_1, properties {soundWaveAssetPtr=&/Kunai_BurningStart_01}}
 {id=WavePlayer_2, properties {bLooping=true, soundWaveAssetPtr=&/Kunai_BurningLoop_01}}
 {id=WavePlayer_3, properties {bLooping=true, soundWaveAssetPtr=&/Kunai_Ambient_Loop_02}}
]
properties {
 attenuationSettings=/Game/Sounds/Attenuation/Fort_Explosive_Fuse_Ambient_Att
 bOverrideAttenuation=true
 bOverrideConcurrency=true
 maxDistance=2800
 soundClassObject=/Game/Sounds/FortSoundClassesAndModes/Fort_Explosives
 volumeMultiplier=0.2
 attenuationOverrides {falloffDistance=2500, stereoSpread=1, attenuationShapeExtents {x=300}}
 concurrencyOverrides {maxCount=8}
}
