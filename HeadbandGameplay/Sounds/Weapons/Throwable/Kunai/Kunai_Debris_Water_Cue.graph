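# Sound cue for the kunai throwable's water debris, playing one of four water impact sounds.
# A Random node picks among the equally weighted waves and feeds a Modulator at the root.
# It plays at reduced volume with close water-explosion attenuation and the large explosives sound class.
# It references a weapon explosion concurrency rule, alongside override settings that stop the oldest instance.
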SoundCue /HeadbandGameplay/Sounds/Weapons/Throwable/Kunai/Kunai_Debris_Water_Cue
root=Modulator_2
nodes [
 {id=Modulator_2, children [Random_2]}
 {id=Random_2, children [WavePlayer_11, WavePlayer_12, WavePlayer_13, WavePlayer_14], properties {weights [1,1,1,1]}}
 {id=WavePlayer_11, properties {soundWaveAssetPtr=&/Kunai_ImpactWater04}}
 {id=WavePlayer_12, properties {soundWaveAssetPtr=&/Kunai_ImpactWater01}}
 {id=WavePlayer_13, properties {soundWaveAssetPtr=&/Kunai_ImpactWater02}}
 {id=WavePlayer_14, properties {soundWaveAssetPtr=&/Kunai_ImpactWater03}}
]
properties {
 attenuationSettings=/Game/Sounds/Attenuation/Fort_Explosion_Water_Close_Athena_Att
 maxDistance=7241.2505
 soundClassObject=/Game/Sounds/FortSoundClassesAndModes/Fort_Explosives_Large_PSM
 volumeMultiplier=0.5
 concurrencyOverrides {maxCount=4, resolutionRule=StopOldest, volumeScale=0.5}
 concurrencySet [/Game/Sounds/Concurrency_Rules/Weapon_Explosion_Conc]
}
