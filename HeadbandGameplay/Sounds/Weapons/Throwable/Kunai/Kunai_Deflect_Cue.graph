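# Sound cue for the deflect sound of the Kunai throwable weapon.
# A random node picks one of three equally weighted deflect waves and feeds it through a modulator for variation.
# It plays in the impacts sound class with raised volume and pitch, overriding attenuation with natural-sound falloff and capping concurrent instances.
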SoundCue /HeadbandGameplay/Sounds/Weapons/Throwable/Kunai/Kunai_Deflect_Cue
root=Modulator_0
nodes [
 {id=Modulator_0, children [Random_0]}
 {id=Random_0, children [WavePlayer_0,WavePlayer_1,WavePlayer_3], properties {weights [1,1,1]}}
 {id=WavePlayer_0, properties {soundWaveAssetPtr=&/Kunai_Deflect_01}}
 {id=WavePlayer_1, properties {soundWaveAssetPtr=&/Kunai_Deflect_02}}
 {id=WavePlayer_3, properties {soundWaveAssetPtr=&/Kunai_Deflect_03}}
]
properties {
 attenuationSettings=&Fort_Weapon_Projectile_Medium_Mono_Athena_Att
 bOverrideAttenuation=true
 bOverrideConcurrency=true
 maxDistance=3800
 pitchMultiplier=1.1
 soundClassObject=&Fort_Impacts
 volumeMultiplier=1.4
 attenuationOverrides {distanceAlgorithm=NaturalSound, falloffDistance=3000, attenuationShapeExtents {x=800}}
 concurrencyOverrides {maxCount=4}
}
