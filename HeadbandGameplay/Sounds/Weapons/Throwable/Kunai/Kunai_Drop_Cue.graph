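# Sound cue that plays the drop sound for the Kunai throwable weapon.
# A single Kunai_Drop_01 wave player feeds a modulator that randomizes volume slightly on each play, with an overall cue volume reduction.
# It uses the weapon drop attenuation settings and the loot collect sound class.
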SoundCue /HeadbandGameplay/Sounds/Weapons/Throwable/Kunai/Kunai_Drop_Cue
root=Modulator_0
nodes [
 {id=Modulator_0, children [WavePlayer_1], properties {volumeMax=0.8, volumeMin=0.7}}
 {id=WavePlayer_1, properties {soundWaveAssetPtr=&/Kunai_Drop_01}}
]
properties {
 attenuationSettings=&Fort_Weapon_Drop_Att
 maxDistance=3000
 soundClassObject=&Fort_Loot_Collect
 volumeMultiplier=0.8
}
