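# Sound cue that plays the equip sound for the Kunai throwable weapon.
# A single wave player for Kunai_Equip_01 feeds a modulator that randomizes volume slightly below full level.
# It uses close-range prop attenuation and the player foley sound class.
# Its concurrency override prevents new instances from starting once the limit is reached.
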SoundCue /HeadbandGameplay/Sounds/Weapons/Throwable/Kunai/Kunai_Equip_Cue
root=Modulator_1
nodes [
 {id=Modulator_1, children [WavePlayer_0], properties {volumeMax=0.85, volumeMin=0.75}}
 {id=WavePlayer_0, properties {soundWaveAssetPtr=&/Kunai_Equip_01}}
]
properties {
 attenuationSettings=/Game/Sounds/Attenuation/Fort_Prop_Close_Att
 bOverrideConcurrency=true
 maxDistance=1800
 soundClassObject=&Fort_Player_Foley
 volumeMultiplier=1
 concurrencyOverrides {resolutionRule=PreventNew}
}
