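# Sound cue for the Kunai throwable's explosion, layered by distance into close, medium and distant sounds.
# Each layer randomly picks one of three waves, varies pitch in a range that drops with distance, then passes through attenuation and a sound class override.
# A mixer combines the layers at lower volume for farther ones; only the close layer overrides its shared attenuation asset.
# Concurrency allows one instance per owner, stopping the oldest.
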SoundCue /HeadbandGameplay/Sounds/Weapons/Throwable/Kunai/Kunai_Explode_Cue
root=Mixer_4
nodes [
 {
  id=Attenuation_0
  children [Modulator_2]
  properties {
   attenuationSettings=/Game/Sounds/Attenuation/Fort_Explosion_Distant_Athena_Att
   attenuationOverrides {falloffDistance=3000, attenuationShapeExtents {x=1000}}
  }
 }
 {
  id=Attenuation_4
  children [Modulator_6]
  properties {
   attenuationSettings=/Game/Sounds/Attenuation/Fort_Explosion_Close_Athena_Att
   bOverrideAttenuation=true
   attenuationOverrides {falloffDistance=5106.2798, stereoSpread=2084.1587, attenuationShapeExtents {x=2254.7803}}
  }
 }
 {
  id=Attenuation_5
  children [Modulator_5]
  properties {
   attenuationSettings=/Game/Sounds/Attenuation/Fort_Explosion_Medium_Athena_Att
   attenuationOverrides {falloffDistance=3000, attenuationShapeExtents {x=1000}}
  }
 }
 {id=Mixer_4, children [SoundClass_3,SoundClass_2,SoundClass_1], properties {inputVolume [1,0.7,0.4]}}
 {id=Modulator_2, children [Random_6], properties {pitchMax=0.9, pitchMin=0.7}}
 {id=Modulator_5, children [Random_5], properties {pitchMax=1, pitchMin=0.8}}
 {id=Modulator_6, children [Random_3], properties {pitchMax=1.1, pitchMin=0.9}}
 {id=Random_3, children [WavePlayer_7, WavePlayer_8, WavePlayer_15], properties {weights [1,1,1]}}
 {id=Random_5, children [WavePlayer_16, WavePlayer_17, WavePlayer_18], properties {weights [1,1,1]}}
 {id=Random_6, children [WavePlayer_19, WavePlayer_20, WavePlayer_21], properties {weights [1,1,1]}}
 {
  id=SoundClass_1
  children [Attenuation_0]
  properties {soundClassOverride=&Fort_WeaponRanged_Fire_Distant}
 }
 {
  id=SoundClass_2
  children [Attenuation_5]
  properties {soundClassOverride=&Fort_WeaponRanged_Fire_Distant}
 }
 {
  id=SoundClass_3
  children [Attenuation_4]
  properties {soundClassOverride=&Fort_WeaponRanged_Fire_3P_PSM}
 }
 {id=WavePlayer_15, properties {soundWaveAssetPtr=&/Kunai_Explode_03}}
 {id=WavePlayer_16, properties {soundWaveAssetPtr=&/Kunai_Explode_01_Medium}}
 {id=WavePlayer_17, properties {soundWaveAssetPtr=&/Kunai_Explode_02_Medium}}
 {id=WavePlayer_18, properties {soundWaveAssetPtr=&/Kunai_Explode_03_Medium}}
 {id=WavePlayer_19, properties {soundWaveAssetPtr=&/Kunai_Explode_01_Distant}}
 {id=WavePlayer_20, properties {soundWaveAssetPtr=&/Kunai_Explode_02_Distant}}
 {id=WavePlayer_21, properties {soundWaveAssetPtr=&/Kunai_Explode_03_Distant}}
 {id=WavePlayer_7, properties {soundWaveAssetPtr=&/Kunai_Explode_01}}
 {id=WavePlayer_8, properties {soundWaveAssetPtr=&/Kunai_Explode_02}}
]
properties {
 bOverrideConcurrency=true
 maxDistance=27500
 pitchMultiplier=1.1
 soundClassObject=/Game/Sounds/FortSoundClassesAndModes/Fort_Explosives_Large_PSM
 volumeMultiplier=0.7076
 concurrencyOverrides {bLimitToOwner=true, maxCount=1, resolutionRule=StopOldest}
}
