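# Sound cue for a thrown kunai striking a body, playing a flesh impact sound.
# A Random node picks one of four equally weighted flesh impact waves, and a Modulator above it slightly varies pitch on each play.
# Overridden attenuation adds distance falloff with low-pass filtering, and overridden concurrency caps simultaneous instances by stopping the oldest.
# It routes to the third-person ranged weapon fire sound class.
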SoundCue /HeadbandGameplay/Sounds/Weapons/Throwable/Kunai/Kunai_ImpactBody_Cue
root=Modulator_0
nodes [
 {id=Modulator_0, children [Random_0], properties {pitchMax=1.1, pitchMin=0.9}}
 {id=Random_0, children [WavePlayer_4,WavePlayer_5,WavePlayer_6,WavePlayer_7], properties {weights [1,1,1,1]}}
 {id=WavePlayer_4, properties {soundWaveAssetPtr=&/Kunai_ImpactFlesh_01}}
 {id=WavePlayer_5, properties {soundWaveAssetPtr=&/Kunai_ImpactFlesh_02}}
 {id=WavePlayer_6, properties {soundWaveAssetPtr=&/Kunai_ImpactFlesh_03}}
 {id=WavePlayer_7, properties {soundWaveAssetPtr=&/Kunai_ImpactFlesh_04}}
]
properties {
 bOverrideAttenuation=true
 bOverrideConcurrency=true
 maxDistance=4200
 soundClassObject=&Fort_WeaponRanged_Fire_3P_PSM
 volumeMultiplier=1
 attenuationOverrides {falloffDistance=3400, lPFRadiusMax=3600, lPFRadiusMin=1800, attenuationShapeExtents {x=800}}
 concurrencyOverrides {maxCount=4, resolutionRule=StopOldest}
}
