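# Sound cue for a thrown kunai impacting a surface.
# It picks one of three equally weighted impact waves and passes it through a modulator that slightly randomizes pitch, on top of a lowered base pitch and boosted volume.
# The cue uses the impacts sound class and overrides attenuation, with low-pass filtering over distance, and concurrency to limit simultaneous plays.
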SoundCue /HeadbandGameplay/Sounds/Weapons/Throwable/Kunai/Kunai_ImpactSurface_Cue
root=Modulator_0
nodes [
 {id=Modulator_0, children [Random_0], properties {pitchMax=1.1, pitchMin=0.9}}
 {id=Random_0, children [WavePlayer_1,WavePlayer_6,WavePlayer_7], properties {weights [1,1,1]}}
 {id=WavePlayer_1, properties {soundWaveAssetPtr=&/Kunai_ImpactSurface_01}}
 {id=WavePlayer_6, properties {soundWaveAssetPtr=&/Kunai_ImpactSurface_02}}
 {id=WavePlayer_7, properties {soundWaveAssetPtr=&/Kunai_ImpactSurface_03}}
]
properties {
 attenuationSettings=&Fort_Weapon_Melee_Sml_Att
 bOverrideAttenuation=true
 bOverrideConcurrency=true
 maxDistance=4400
 pitchMultiplier=0.9
 soundClassObject=&Fort_Impacts
 volumeMultiplier=1.5
 attenuationOverrides {bAttenuateWithLPF=true, lPFRadiusMax=3600, lPFRadiusMin=1800, attenuationShapeExtents {x=800}}
 concurrencyOverrides {maxCount=4}
}
