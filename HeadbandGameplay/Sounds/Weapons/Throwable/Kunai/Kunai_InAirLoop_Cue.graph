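# Sound cue for the looping in-air sound of a thrown Kunai, classed as a ranged-weapon projectile effect.
# A single looping wave player feeds an Attenuation node that uses the close-range Athena projectile attenuation.
# The cue itself references the Batarang variant of that attenuation and applies slight pitch-up and volume-down multipliers.
# Its concurrency override prevents new instances from starting once the limit is reached.
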SoundCue /HeadbandGameplay/Sounds/Weapons/Throwable/Kunai/Kunai_InAirLoop_Cue
root=Attenuation_0
nodes [
 {id=Attenuation_0, children [WavePlayer_1], properties {attenuationSettings=&Fort_Weapon_Projectile_Close_Athena_Att}}
 {id=WavePlayer_1, properties {bLooping=true, soundWaveAssetPtr=&/Kunai_InAirLoop_01}}
]
properties {
 attenuationSettings=/Game/Sounds/Attenuation/Fort_Weapon_Projectile_Close_Athena_Batarang_Att
 bOverrideConcurrency=true
 maxDistance=4950
 pitchMultiplier=1.1
 soundClassObject=&Fort_WeaponRanged_Projectile_Effects
 volumeMultiplier=0.85
 attenuationOverrides {falloffDistance=5000, attenuationShapeExtents {x=1000}}
 concurrencyOverrides {resolutionRule=PreventNew}
}
