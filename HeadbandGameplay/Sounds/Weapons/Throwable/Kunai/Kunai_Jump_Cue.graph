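# Sound cue for the Kunai throwable's jump sound in the Headband gameplay plugin.
# A random node picks one of two equally weighted jump waves, which a modulator varies with a slight pitch drop and volume boost.
# It uses the first-person player foley sound class and borrows player-landing attenuation.
# A concurrency override allows one instance per owner, blocking new plays while one is active.
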SoundCue /HeadbandGameplay/Sounds/Weapons/Throwable/Kunai/Kunai_Jump_Cue
root=Modulator_0
nodes [
 {id=Modulator_0, children [Random_3], properties {pitchMax=1, pitchMin=0.8, volumeMax=1.2, volumeMin=1}}
 {id=Random_3, children [WavePlayer_2,WavePlayer_4], properties {weights [1,1]}}
 {id=WavePlayer_2, properties {soundWaveAssetPtr=&/Kunai_Jump_01}}
 {id=WavePlayer_4, properties {soundWaveAssetPtr=&/Kunai_Jump_03}}
]
properties {
 attenuationSettings=/Game/Sounds/Attenuation/Fort_Player_Land_Athena_Att
 bOverrideConcurrency=true
 maxDistance=2097152
 soundClassObject=&Fort_Player_Foley_1P_Athena
 volumeMultiplier=1.7
 concurrencyOverrides {bLimitToOwner=true, maxCount=1, resolutionRule=PreventNew}
}
