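# Sound cue for picking up the Kunai throwable weapon in the HeadbandGameplay content.
# It plays a single pickup wave through a Modulator that randomizes pitch slightly downward for variation.
# It uses the shared weapon pickup attenuation and loot collect sound class at reduced volume, and a concurrency override prevents new instances once the limit is reached.
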SoundCue /HeadbandGameplay/Sounds/Weapons/Throwable/Kunai/Kunai_Pickup_Cue
root=Modulator_0
nodes [
 {id=Modulator_0, children [WavePlayer_0], properties {pitchMax=1, pitchMin=0.9}}
 {id=WavePlayer_0, properties {soundWaveAssetPtr=&/Kunai_Pickup_01}}
]
properties {
 attenuationSettings=&Fort_Weapon_Pickup_Att
 bOverrideConcurrency=true
 maxDistance=3500
 soundClassObject=&Fort_Loot_Collect
 volumeMultiplier=0.4308
 concurrencyOverrides {resolutionRule=PreventNew}
}
