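# Sound cue for the Kunai throwable's pre-throw sound, built from a close layer and a distant layer.
# Each layer randomly picks one of three equally weighted waves and varies its volume slightly, with the distant layer set quieter.
# Each layer then gets its own projectile attenuation and ranged-weapon fire sound class before both are mixed.
# A concurrency override prevents new instances once the limit is reached.
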SoundCue /HeadbandGameplay/Sounds/Weapons/Throwable/Kunai/Kunai_PreThrow_Cue
root=Mixer_0
nodes [
 {
  id=Attenuation_0
  children [Modulator_1]
  properties {attenuationSettings=&Fort_Weapon_Projectile_Medium_Mono_Athena_Att}
 }
 {id=Attenuation_1, children [Modulator_0], properties {attenuationSettings=&Fort_Weapon_Projectile_Close_Athena_Att}}
 {id=Mixer_0, children [SoundClass_1,SoundClass_0], properties {inputVolume [0.8,0.8]}}
 {id=Modulator_0, children [Random_0], properties {volumeMax=1, volumeMin=0.9}}
 {id=Modulator_1, children [Random_3], properties {volumeMax=0.8, volumeMin=0.7}}
 {id=Random_0, children [WavePlayer_0,WavePlayer_1,WavePlayer_7], properties {weights [1,1,1]}}
 {id=Random_3, children [WavePlayer_10, WavePlayer_11, WavePlayer_13], properties {weights [1,1,1]}}
 {
  id=SoundClass_0
  children [Attenuation_0]
  properties {soundClassOverride=&Fort_WeaponRanged_Fire_Distant}
 }
 {
  id=SoundClass_1
  children [Attenuation_1]
  properties {soundClassOverride=&Fort_WeaponRanged_Fire_3P_PSM}
 }
 {id=WavePlayer_0, properties {soundWaveAssetPtr=&/Kunai_PreThrow_01}}
 {id=WavePlayer_1, properties {soundWaveAssetPtr=&/Kunai_PreThrow_02}}
 {id=WavePlayer_10, properties {soundWaveAssetPtr=&/Kunai_PreThrow_Distant_01}}
 {id=WavePlayer_11, properties {soundWaveAssetPtr=&/Kunai_PreThrow_Distant_02}}
 {id=WavePlayer_13, properties {soundWaveAssetPtr=&/Kunai_PreThrow_Distant_03}}
 {id=WavePlayer_7, properties {soundWaveAssetPtr=&/Kunai_PreThrow_03}}
]
properties {
 bOverrideConcurrency=true
 maxDistance=10579.2871
 volumeMultiplier=0.6
 attenuationOverrides {falloffDistance=5000, attenuationShapeExtents {x=1000}}
 concurrencyOverrides {resolutionRule=PreventNew}
}
