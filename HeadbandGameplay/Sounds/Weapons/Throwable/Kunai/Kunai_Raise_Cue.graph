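# Sound cue for the kunai throwable's raise foley, played as a player foley sound.
# A random node picks one of three slow-raise waves with equal weight, then a modulator slightly lowers and varies its pitch and volume.
# It uses close-prop attenuation and overrides concurrency to block new instances once the limit is reached, with an overall volume boost.
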SoundCue /HeadbandGameplay/Sounds/Weapons/Throwable/Kunai/Kunai_Raise_Cue
root=Modulator_1
nodes [
 {id=Modulator_1, children [Random_0], properties {pitchMax=0.95, pitchMin=0.9, volumeMax=0.85, volumeMin=0.75}}
 {id=Random_0, children [WavePlayer_1,WavePlayer_2,WavePlayer_3], properties {weights [1,1,1]}}
 {id=WavePlayer_1, properties {soundWaveAssetPtr=&/Kunai_Raise_Slow_01}}
 {id=WavePlayer_2, properties {soundWaveAssetPtr=&/Kunai_Raise_Slow_02}}
 {id=WavePlayer_3, properties {soundWaveAssetPtr=&/Kunai_Raise_Slow_03}}
]
properties {
 attenuationSettings=/Game/Sounds/Attenuation/Fort_Prop_Close_Att
 bOverrideConcurrency=true
 maxDistance=1800
 soundClassObject=&Fort_Player_Foley
 volumeMultiplier=1.3
 concurrencyOverrides {resolutionRule=PreventNew}
}
