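# Sound cue for the Kunai throwable's target-off sound in HeadbandGameplay.
# A single wave player plays a reused BlackMonday bang target-off sample at reduced volume, with no mixing or randomization.
# It routes through the gadgets and items sound class, uses weapon pickup attenuation, and overrides concurrency to stop the oldest instance.
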SoundCue /HeadbandGameplay/Sounds/Weapons/Throwable/Kunai/Kunai_TargetOff_Cue
root=WavePlayer_0
nodes [{id=WavePlayer_0, properties {soundWaveAssetPtr=/Game/Sounds/Fort_Weapons/BlackMonday/Bangs/BlackMonday_Bang_TargetOff_01}}]
properties {
 attenuationSettings=&Fort_Weapon_Pickup_Att
 bOverrideConcurrency=true
 maxDistance=3500
 soundClassObject=&Fort_Gadgets_Items
 volumeMultiplier=0.4
 concurrencyOverrides {resolutionRule=StopOldest, volumeScaleAttackTime=1}
}
