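# Sound cue for the throwable Kunai, named for its target-on moment and built from pre-throw sounds.
# A Random node picks one of three equally weighted Kunai_PreThrow wave variants each play.
# Playback is quieted, routed to the gadget items sound class and spatialized with shared weapon pickup attenuation.
# A concurrency override prevents new instances once the voice limit is reached instead of cutting off playing ones.
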SoundCue /HeadbandGameplay/Sounds/Weapons/Throwable/Kunai/Kunai_TargetOn_Cue
root=Random_0
nodes [
 {id=Random_0, children [WavePlayer_1,WavePlayer_2,WavePlayer_3], properties {weights [1,1,1]}}
 {id=WavePlayer_1, properties {soundWaveAssetPtr=&/Kunai_PreThrow_01}}
 {id=WavePlayer_2, properties {soundWaveAssetPtr=&/Kunai_PreThrow_02}}
 {id=WavePlayer_3, properties {soundWaveAssetPtr=&/Kunai_PreThrow_03}}
]
properties {
 attenuationSettings=&Fort_Weapon_Pickup_Att
 bOverrideConcurrency=true
 maxDistance=3500
 soundClassObject=&Fort_Gadgets_Items
 volumeMultiplier=0.4
 concurrencyOverrides {resolutionRule=PreventNew}
}
