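# First-person sound cue for throwing a kunai weapon, routed to the first-person ranged weapon fire sound class.
# A quality-level node sends its first two levels to a random pick of four throw waves and its last two to a random pick of two.
# Each branch passes through a modulator that slightly varies volume.
# Projectile close-range attenuation and a shared first-person gunfire concurrency set govern playback.
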SoundCue /HeadbandGameplay/Sounds/Weapons/Throwable/Kunai/Kunai_Throw_1P_Cue
root=QualityLevel_0
nodes [
 {id=Modulator_0, children [Random_3], properties {volumeMax=1, volumeMin=0.9}}
 {id=Modulator_2, children [Random_8], properties {volumeMax=1, volumeMin=0.9}}
 {id=QualityLevel_0, children [Modulator_0,Modulator_0,Modulator_2,Modulator_2]}
 {id=Random_3, children [WavePlayer_14, WavePlayer_18, WavePlayer_19, WavePlayer_20], properties {weights [1,1,1,1]}}
 {id=Random_8, children [WavePlayer_26, WavePlayer_27], properties {weights [1,1]}}
 {id=WavePlayer_14, properties {soundWaveAssetPtr=&/Kunai_Throw_01}}
 {id=WavePlayer_18, properties {soundWaveAssetPtr=&/Kunai_Throw_02}}
 {id=WavePlayer_19, properties {soundWaveAssetPtr=&/Kunai_Throw_03}}
 {id=WavePlayer_20, properties {soundWaveAssetPtr=&/Kunai_Throw_04}}
 {id=WavePlayer_26, properties {soundWaveAssetPtr=&Kunai_Throw_01}}
 {id=WavePlayer_27, properties {soundWaveAssetPtr=&Kunai_Throw_02}}
]
properties {
 attenuationSettings=&Fort_Weapon_Projectile_Close_Athena_Att
 maxDistance=4950
 soundClassObject=&Fort_WeaponRanged_Fire_1P_PSM
 volumeMultiplier=0.84
 attenuationOverrides {falloffDistance=5000, attenuationShapeExtents {x=1000}}
 concurrencyOverrides {resolutionRule=PreventNew}
 concurrencySet [&Weapon_Gunfire_FirstPerson_Conc]
}
