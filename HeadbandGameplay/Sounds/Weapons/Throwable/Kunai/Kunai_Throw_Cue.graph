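# Sound cue for the kunai throw, layering a close throw sound with a distant layer under gunfire concurrency limits.
# Each layer randomizes waves through volume modulation, its own attenuation and sound class (3P fire or distant), then joins a mixer.
# The distant layer reuses BlackMonday throw waves, pitched up.
# A quality node sends its first two levels to four-variation mixes and its last two to two-variation mixes.
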SoundCue /HeadbandGameplay/Sounds/Weapons/Throwable/Kunai/Kunai_Throw_Cue
root=QualityLevel_1
nodes [
 {
  id=Attenuation_0
  children [Modulator_1]
  properties {attenuationSettings=&Fort_Weapon_Projectile_Medium_Mono_Athena_Att}
 }
 {id=Attenuation_1, children [Modulator_0], properties {attenuationSettings=&Fort_Weapon_Projectile_Close_Athena_Att}}
 {
  id=Attenuation_2
  children [Modulator_3]
  properties {attenuationSettings=&Fort_Weapon_Projectile_Medium_Mono_Athena_Att}
 }
 {id=Attenuation_7, children [Modulator_2], properties {attenuationSettings=&Fort_Weapon_Projectile_Close_Athena_Att}}
 {id=Mixer_0, children [SoundClass_1,SoundClass_0], properties {inputVolume [0.8,0.8]}}
 {id=Mixer_2, children [SoundClass_7,SoundClass_6], properties {inputVolume [0.8,0.8]}}
 {id=Modulator_0, children [Random_3], properties {volumeMax=1, volumeMin=0.9}}
 {id=Modulator_1, children [Random_0], properties {pitchMax=1.4, pitchMin=1.3, volumeMax=0.8, volumeMin=0.7}}
 {id=Modulator_2, children [Random_8], properties {volumeMax=1, volumeMin=0.9}}
 {id=Modulator_3, children [Random_5], properties {pitchMax=1.4, pitchMin=1.3, volumeMax=0.8, volumeMin=0.7}}
 {id=QualityLevel_1, children [Mixer_0,Mixer_0,Mixer_2,Mixer_2]}
 {id=Random_0, children [WavePlayer_0, WavePlayer_1, WavePlayer_7, WavePlayer_13], properties {weights [1,1,1,1]}}
 {id=Random_3, children [WavePlayer_14, WavePlayer_18, WavePlayer_19, WavePlayer_20], properties {weights [1,1,1,1]}}
 {id=Random_5, children [WavePlayer_21, WavePlayer_22], properties {weights [1,1]}}
 {id=Random_8, children [WavePlayer_26, WavePlayer_27], properties {weights [1,1]}}
 {
  id=SoundClass_0
  children [Attenuation_0]
  properties {soundClassOverride=&Fort_WeaponRanged_Fire_Distant}
 }
 {
  id=SoundClass_1
  children [Attenuation_1]
  properties {soundClassOverride=&Fort_WeaponRanged_Fire_3P_PSM}
 }
 {
  id=SoundClass_6
  children [Attenuation_2]
  properties {soundClassOverride=&Fort_WeaponRanged_Fire_Distant}
 }
 {
  id=SoundClass_7
  children [Attenuation_7]
  properties {soundClassOverride=&Fort_WeaponRanged_Fire_3P_PSM}
 }
 {id=WavePlayer_0, properties {soundWaveAssetPtr=/Game/Sounds/Fort_Weapons/BlackMonday/Bangs/BlackMonday_Bang_Throw_Distant_01}}
 {id=WavePlayer_1, properties {soundWaveAssetPtr=/Game/Sounds/Fort_Weapons/BlackMonday/Bangs/BlackMonday_Bang_Throw_Distant_02}}
 {id=WavePlayer_13, properties {soundWaveAssetPtr=/Game/Sounds/Fort_Weapons/BlackMonday/Bangs/BlackMonday_Bang_Throw_Distant_04}}
 {id=WavePlayer_14, properties {soundWaveAssetPtr=&/Kunai_Throw_01}}
 {id=WavePlayer_18, properties {soundWaveAssetPtr=&/Kunai_Throw_02}}
 {id=WavePlayer_19, properties {soundWaveAssetPtr=&/Kunai_Throw_03}}
 {id=WavePlayer_20, properties {soundWaveAssetPtr=&/Kunai_Throw_04}}
 {id=WavePlayer_21, properties {soundWaveAssetPtr=&BlackMonday_Bang_Throw_Distant_01}}
 {id=WavePlayer_22, properties {soundWaveAssetPtr=&BlackMonday_Bang_Throw_Distant_02}}
 {id=WavePlayer_26, properties {soundWaveAssetPtr=&Kunai_Throw_01}}
 {id=WavePlayer_27, properties {soundWaveAssetPtr=&Kunai_Throw_02}}
 {id=WavePlayer_7, properties {soundWaveAssetPtr=/Game/Sounds/Fort_Weapons/BlackMonday/Bangs/BlackMonday_Bang_Throw_Distant_03}}
]
properties {
 maxDistance=10579.2871
 attenuationOverrides {falloffDistance=5000, attenuationShapeExtents {x=1000}}
 concurrencyOverrides {resolutionRule=PreventNew}
 concurrencySet [&Weapon_Gunfire_Local_Conc, &Weapon_Gunfire_Global_Conc]
}
