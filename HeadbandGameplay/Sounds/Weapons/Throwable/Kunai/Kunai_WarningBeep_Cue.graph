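# Sound cue for the warning beep of the throwable Kunai weapon.
# A Random node picks one of three equally weighted BlackMonday bang warning-beep waves and plays it at a strongly reduced volume.
# It references the gadget-working attenuation asset but overrides attenuation and concurrency locally, limiting range and simultaneous instances.
# It routes through a third-person ranged weapon fire sound class.
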SoundCue /HeadbandGameplay/Sounds/Weapons/Throwable/Kunai/Kunai_WarningBeep_Cue
root=Random_0
nodes [
 {id=Random_0, children [WavePlayer_1,WavePlayer_2,WavePlayer_3], properties {weights [1,1,1]}}
 {id=WavePlayer_1, properties {soundWaveAssetPtr=/Game/Sounds/Fort_Weapons/BlackMonday/Bangs/BlackMonday_Bang_WarningBeep_01}}
 {id=WavePlayer_2, properties {soundWaveAssetPtr=/Game/Sounds/Fort_Weapons/BlackMonday/Bangs/BlackMonday_Bang_WarningBeep_02}}
 {id=WavePlayer_3, properties {soundWaveAssetPtr=/Game/Sounds/Fort_Weapons/BlackMonday/Bangs/BlackMonday_Bang_WarningBeep_03}}
]
properties {
 attenuationSettings=/Game/Sounds/Attenuation/Fort_GadgetWorking_Att
 bOverrideAttenuation=true
 bOverrideConcurrency=true
 maxDistance=1536
 soundClassObject=&Fort_WeaponRanged_Fire_3P_PSM
 volumeMultiplier=0.1
 attenuationOverrides {falloffDistance=1024, attenuationShapeExtents {x=512}}
 concurrencyOverrides {maxCount=3}
}
