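# Sound cue for the water splash of the throwable kunai weapon, built from borrowed grenade water-impact sounds.
# A random node picks one of three equally weighted waves and feeds a modulator at its default settings.
# The cue plays at reduced volume with grenade-bounce attenuation and a third-person ranged weapon fire sound class.
# It references shared weapon-explosion and item-splash concurrency sets.
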SoundCue /HeadbandGameplay/Sounds/Weapons/Throwable/Kunai/Kunai_Water_Splash_Cue
root=Modulator_1
nodes [
 {id=Modulator_1, children [Random_1]}
 {id=Random_1, children [WavePlayer_0,WavePlayer_1,WavePlayer_4], properties {weights [1,1,1]}}
 {id=WavePlayer_0, properties {soundWaveAssetPtr=/Game/Sounds/Fort_GamePlay_Sounds/Explosives/Grenade_Impact_Water_01}}
 {id=WavePlayer_1, properties {soundWaveAssetPtr=/Game/Sounds/Fort_GamePlay_Sounds/Explosives/Grenade_Impact_Water_02}}
 {id=WavePlayer_4, properties {soundWaveAssetPtr=/Game/Sounds/Fort_GamePlay_Sounds/Explosives/Grenade_Impact_Water_03}}
]
properties {
 attenuationSettings=/Game/Sounds/Attenuation/Fort_Grenade_Bounce_Att
 maxDistance=5500
 soundClassObject=&Fort_WeaponRanged_Fire_3P_PSM
 volumeMultiplier=0.55
 concurrencyOverrides {maxCount=4, resolutionRule=StopOldest, volumeScale=0.5}
 concurrencySet [/Game/Sounds/Concurrency_Rules/Weapon_Explosion_Conc, /Game/Sounds/Concurrency_Rules/Item_Splash_Conc]
}
